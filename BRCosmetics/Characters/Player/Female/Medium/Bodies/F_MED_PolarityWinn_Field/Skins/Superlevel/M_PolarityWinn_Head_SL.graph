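# Head material for the Superlevel skin of the PolarityWinn Field character body, recoloring the hair.
# It takes material attributes from the shared character parent PreFX function and lerps Base Color toward HairColor_SL using the HairMask blue channel.
# The modified attributes then feed the parent PostFX function.
# Separate hue-shift and desaturate-multiply tint branches are left unconnected.
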Material /BRCosmetics/Characters/Player/Female/Medium/Bodies/F_MED_PolarityWinn_Field/Skins/Superlevel/M_PolarityWinn_Head_SL
graph {
 properties {
  bUseMaterialAttributes=true
  bUsedWithMeshDeformer=true
  bUsedWithMorphTargets=true
  bUsedWithSkeletalMesh=true
  pixelDepthOffsetMode=PDOM_Legacy
  shadingModel=MSM_FromMaterialExpression
  usedShadingModels=Default Lit | Subsurface | Clear Coat | Subsurface Profile | Cloth
 }
 nodes [
  n1 = MaterialFunctionCall() {materialFunction=&MF_CharacterParent_2023_PreFX}
  n2 = GetMaterialAttributes(None=@n1.Result) {attributeGetTypes [#1]}
  n3 = VectorParameter() {parameterName=HairColor_SL, defaultValue=<rgba 0.5222 1 0.4857 1>}
  n4 = TextureSampleParameter2D() {
   parameterName=HairMask
   samplerSource=SSM_Wrap_WorldGroupSettings
   samplerType=SAMPLERTYPE_Masks
   texture=/BRCosmetics/Characters/Player/Female/Medium/Bodies/F_MED_PolarityWinn_Field/Textures/T_PolarityWinn_Field_Head_SL_FX2
  }
  MaterialFunctionCall(
   Material=SetMaterialAttributes!(MaterialAttributes=@n2.MaterialAttributes, Base Color=LinearInterpolate(A=@n2.Base Color, B=@n3.RGB, Alpha=@n4.B)) {attributeSetTypes [#1]}
   M=@n1.M
   Pre Skinned Local Position=@n1.M
   Distance Blend=@n1.M
   Fresnel=@n1.M
  ) {materialFunction=&MF_CharacterParent_2023_PostFX}
  MaterialFunctionCall(Hue Shift Percentage=Multiply(A=@n4.B, B=ScalarParameter() {parameterName=Hair_HueShift}), Texture=@n2.Base Color) {materialFunction=&HueShift}
  Multiply(A=Desaturation(None=@n2.Base Color, Fraction=@n4.B), B=Multiply(A=@n4.B, B=@n3.RGB))
 ]
}
